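# Physics asset for the pack-equipped variant of the male medium YogaPatio body, simulating its jacket with capsule bodies on dyn_jacket bone chains.
# Front, side-front, side, side-back and left-only back chains hang from tiny "mstr" anchor capsules, with large spine_03 and spine_04 capsules as unconstrained torso volumes.
# Lower joints swing on one axis with minimal twist.
# Front and side roots swing on two axes, while back and side-back roots are locked.
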PhysicsAsset /BRCosmetics/Characters/Player/Male/Medium/Bodies/M_MED_YogaPatio/Meshes/M_MED_YogaPatio_PackEquipped_Physics
bodies {
 dyn_jacket_bk_01_l [{shapeType=Capsule, length=16.0118, radius=2.9129, shapeName=dyn_jacket_bk_01_l_capsule, center=<2.3925 -1.8767 0>}]
 dyn_jacket_bk_01_mstr_l [{shapeType=Capsule, length=0.505, radius=0.505, shapeName=dyn_jacket_bk_01_mstr_l_capsule, rotation=<rot 0 -90 0>}]
 dyn_jacket_bk_02_l [{shapeType=Capsule, length=15.3762, radius=3.3854, shapeName=dyn_jacket_bk_02_l_capsule, center=<2.3925 -1.8767 0>}]
 dyn_jacket_fr_01_l [{shapeType=Capsule, length=16.3432, radius=2.6, shapeName=dyn_jacket_fr_01_l_capsule, center=<2.3925 -1.8767 0>}]
 dyn_jacket_fr_01_mstr_l [{shapeType=Capsule, length=0.505, radius=0.505, shapeName=dyn_jacket_fr_01_mstr_l_capsule, rotation=<rot 0 -90 0>}]
 dyn_jacket_fr_01_mstr_r [{shapeType=Capsule, length=0.505, radius=0.505, shapeName=dyn_jacket_fr_01_mstr_r_capsule, rotation=<rot 0 -90 0>}]
 dyn_jacket_fr_01_r [{shapeType=Capsule, length=16.3432, radius=2.6, shapeName=dyn_jacket_fr_01_r_capsule, center=<2.3925 -1.8767 0>}]
 dyn_jacket_fr_02_l [{shapeType=Capsule, length=16.3432, radius=2.6, shapeName=dyn_jacket_fr_02_l_capsule, center=<2.3925 -1.8767 0>}]
 dyn_jacket_fr_02_r [{shapeType=Capsule, length=16.3432, radius=2.6, shapeName=dyn_jacket_fr_02_r_capsule, center=<2.3925 -1.8767 0>}]
 dyn_jacket_fr_03_l [{shapeType=Capsule, length=15.5342, radius=2.6, shapeName=dyn_jacket_fr_03_l_capsule, center=<2.3925 -3.7968 0>}]
 dyn_jacket_fr_03_r [{shapeType=Capsule, length=15.5342, radius=2.6, shapeName=dyn_jacket_fr_03_r_capsule, center=<2.3925 -3.7968 0>}]
 dyn_jacket_side_01_l [{
   shapeType=Capsule
   length=13.3218
   radius=5.0734
   shapeName=dyn_jacket_side_01_l_capsule
   center=<2.3925 -1.4121 0>
   rotation=<rot 0.0001 -23.8716 -9.7322>
  }]
 dyn_jacket_side_01_mstr_l [{shapeType=Capsule, length=0.505, radius=0.505, shapeName=dyn_jacket_side_01_mstr_l_capsule, rotation=<rot 0 -90 0>}]
 dyn_jacket_side_01_mstr_r [{shapeType=Capsule, length=0.505, radius=0.505, shapeName=dyn_jacket_side_01_mstr_r_capsule, rotation=<rot 0 -90 0>}]
 dyn_jacket_side_01_r [{shapeType=Capsule, length=12.4859, radius=5.679, shapeName=dyn_jacket_side_01_r_capsule, center=<2.3925 -1.8767 0>}]
 dyn_jacket_side_02_l [{
   shapeType=Capsule
   length=13.6631
   radius=5.0056
   shapeName=dyn_jacket_side_02_l_capsule
   center=<2.3925 -1.8767 0>
   rotation=<rot 0 0 -9.7322>
  }]
 dyn_jacket_side_02_r [{shapeType=Capsule, length=12.5235, radius=5.6107, shapeName=dyn_jacket_side_02_r_capsule, center=<2.3925 -1.8767 0>}]
 dyn_jacket_sidebk_01_l [{
   shapeType=Capsule
   length=13.4433
   radius=4.8076
   shapeName=dyn_jacket_sidebk_01_l_capsule
   center=<2.3925 -1.8767 0>
   rotation=<rot 0 0 4.7643>
  }]
 dyn_jacket_sidebk_01_mstr_l [{shapeType=Capsule, length=0.505, radius=0.505, shapeName=dyn_jacket_sidebk_01_mstr_l_capsule, rotation=<rot 0 -90 0>}]
 dyn_jacket_sidebk_01_mstr_r [{shapeType=Capsule, length=0.505, radius=0.505, shapeName=dyn_jacket_sidebk_01_mstr_r_capsule, rotation=<rot 0 -90 0>}]
 dyn_jacket_sidebk_01_r [{
   shapeType=Capsule
   length=14.6587
   radius=4.9167
   shapeName=dyn_jacket_sidebk_01_r_capsule
   center=<2.3925 -1.8767 0>
   rotation=<rot -0.0001 -1.3601 -7.3525>
  }]
 dyn_jacket_sidebk_02_l [{
   shapeType=Capsule
   length=18.0656
   radius=5.0282
   shapeName=dyn_jacket_sidebk_02_l_capsule
   center=<2.3925 -1.8717 0.1376>
   rotation=<rot 0 0.0001 20.9802>
  }]
 dyn_jacket_sidebk_02_r [{
   shapeType=Capsule
   length=24.5772
   radius=3.5681
   shapeName=dyn_jacket_sidebk_01_r_capsule
   center=<2.3925 -1.8682 0.1782>
   rotation=<rot -0.0001 0.0002 -5.401>
  }]
 dyn_jacket_sidefr_01_l [{shapeType=Capsule, length=15.1865, radius=3.7567, shapeName=dyn_jacket_sidefr_01_l_capsule, center=<2.3925 -1.8767 0>}]
 dyn_jacket_sidefr_01_mstr_l [{shapeType=Capsule, length=0.505, radius=0.505, shapeName=dyn_jacket_sidefr_01_mstr_l_capsule, rotation=<rot 0 -90 0>}]
 dyn_jacket_sidefr_01_mstr_r [{shapeType=Capsule, length=0.505, radius=0.505, shapeName=dyn_jacket_sidefr_01_mstr_r_capsule, rotation=<rot 0 -90 0>}]
 dyn_jacket_sidefr_01_r [{shapeType=Capsule, length=15.0502, radius=3.893, shapeName=dyn_jacket_sidefr_01_r_capsule, center=<2.3925 -1.8767 0>}]
 dyn_jacket_sidefr_02_l [{shapeType=Capsule, length=14.3775, radius=3.7567, shapeName=dyn_jacket_sidefr_02_l_capsule, center=<2.3925 -1.8767 0>}]
 dyn_jacket_sidefr_02_r [{shapeType=Capsule, length=14.2412, radius=3.893, shapeName=dyn_jacket_sidefr_02_r_capsule, center=<2.3925 -1.8767 0>}]
 spine_03 [{
   shapeType=Capsule
   length=4.6324
   radius=13.8557
   shapeName=spine_03_capsule
   center=<-5.3827 1.7073 0>
   rotation=<rot 90 -179.1048 175.3053>
  }]
 spine_04 [{
   shapeType=Capsule
   length=17.7821
   radius=12.6426
   shapeName=spine_04_capsule
   center=<1.9233 2.381 -0.0711>
   rotation=<rot 90 -179.6643 179.9445>
  }]
}
constraints [
 {bone1Name=dyn_jacket_bk_02_l, bone2Name=dyn_jacket_bk_01_l, swing1LimitDegrees=2, swing2LimitDegrees=0, twistLimitDegrees=2}
 {bone1Name=dyn_jacket_fr_02_l, bone2Name=dyn_jacket_fr_01_l, swing1LimitDegrees=20, swing2LimitDegrees=0, twistLimitDegrees=2}
 {bone1Name=dyn_jacket_fr_03_l, bone2Name=dyn_jacket_fr_02_l, swing1LimitDegrees=20, swing2LimitDegrees=0, twistLimitDegrees=2}
 {bone1Name=dyn_jacket_fr_02_r, bone2Name=dyn_jacket_fr_01_r, swing1LimitDegrees=20, swing2LimitDegrees=0, twistLimitDegrees=2}
 {bone1Name=dyn_jacket_fr_03_r, bone2Name=dyn_jacket_fr_02_r, swing1LimitDegrees=20, swing2LimitDegrees=0, twistLimitDegrees=2}
 {bone1Name=dyn_jacket_side_02_l, bone2Name=dyn_jacket_side_01_l, swing1LimitDegrees=40, swing2LimitDegrees=0, twistLimitDegrees=2}
 {bone1Name=dyn_jacket_side_02_r, bone2Name=dyn_jacket_side_01_r, swing1LimitDegrees=40, swing2LimitDegrees=0, twistLimitDegrees=2}
 {bone1Name=dyn_jacket_sidebk_02_l, bone2Name=dyn_jacket_sidebk_01_l, swing1LimitDegrees=20, swing2LimitDegrees=0, twistLimitDegrees=2}
 {bone1Name=dyn_jacket_sidefr_02_l, bone2Name=dyn_jacket_sidefr_01_l, swing1LimitDegrees=20, swing2LimitDegrees=0, twistLimitDegrees=2}
 {bone1Name=dyn_jacket_sidefr_02_r, bone2Name=dyn_jacket_sidefr_01_r, swing1LimitDegrees=20, swing2LimitDegrees=0, twistLimitDegrees=2}
 {bone1Name=dyn_jacket_fr_01_l, bone2Name=dyn_jacket_fr_01_mstr_l, swing1LimitDegrees=25, swing2LimitDegrees=15, twistLimitDegrees=2}
 {bone1Name=dyn_jacket_sidefr_01_l, bone2Name=dyn_jacket_sidefr_01_mstr_l, swing1LimitDegrees=25, swing2LimitDegrees=15, twistLimitDegrees=2}
 {bone1Name=dyn_jacket_side_01_l, bone2Name=dyn_jacket_side_01_mstr_l, swing1LimitDegrees=30, swing2LimitDegrees=20, twistLimitDegrees=2}
 {
  bone1Name=dyn_jacket_sidebk_01_l
  bone2Name=dyn_jacket_sidebk_01_mstr_l
  swing1LimitDegrees=2
  swing1Motion=Locked
  swing2LimitDegrees=2
  swing2Motion=Locked
  twistLimitDegrees=2
  twistMotion=Locked
 }
 {
  bone1Name=dyn_jacket_bk_01_l
  bone2Name=dyn_jacket_bk_01_mstr_l
  swing1LimitDegrees=2
  swing1Motion=Locked
  swing2LimitDegrees=2
  swing2Motion=Locked
  twistLimitDegrees=2
  twistMotion=Locked
 }
 {bone1Name=dyn_jacket_side_01_r, bone2Name=dyn_jacket_side_01_mstr_r, swing1LimitDegrees=20, swing2LimitDegrees=20, twistLimitDegrees=2}
 {bone1Name=dyn_jacket_sidefr_01_r, bone2Name=dyn_jacket_sidefr_01_mstr_r, swing1LimitDegrees=25, swing2LimitDegrees=15, twistLimitDegrees=2}
 {bone1Name=dyn_jacket_fr_01_r, bone2Name=dyn_jacket_fr_01_mstr_r, swing1LimitDegrees=25, swing2LimitDegrees=15, twistLimitDegrees=2}
 {bone1Name=dyn_jacket_sidebk_02_r, bone2Name=dyn_jacket_sidebk_01_r, swing1LimitDegrees=20, swing2LimitDegrees=0, twistLimitDegrees=2}
 {
  bone1Name=dyn_jacket_sidebk_01_r
  bone2Name=dyn_jacket_sidebk_01_mstr_r
  swing1LimitDegrees=2
  swing1Motion=Locked
  swing2LimitDegrees=2
  swing2Motion=Locked
  twistLimitDegrees=2
  twistMotion=Locked
 }
]
