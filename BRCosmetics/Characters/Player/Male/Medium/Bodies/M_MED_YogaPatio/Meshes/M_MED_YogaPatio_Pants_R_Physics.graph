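# Physics asset for the right pant leg of the M_MED_YogaPatio male player cosmetic body, covering its dynamic cloth bones.
# A tiny capsule on calf_r serves as the anchor, with larger capsules on dyn_pants_01_r and dyn_pants_02_r.
# Each pants bone is constrained directly to the calf rather than chained, with narrow swing and tighter twist limits.
# Shape names keep left-side "_l" labels despite the right-side bones.
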PhysicsAsset /BRCosmetics/Characters/Player/Male/Medium/Bodies/M_MED_YogaPatio/Meshes/M_MED_YogaPatio_Pants_R_Physics
bodies {
 calf_r [{shapeType=Capsule, length=0.505, radius=0.505, shapeName=calf_l_capsule, rotation=<rot 0 -90 0>}]
 dyn_pants_01_r [{shapeType=Capsule, length=2.6572, radius=9.5095, shapeName=dyn_pants_01_l_capsule, center=<7.7768 0 0>, rotation=<rot 90 0 0>}]
 dyn_pants_02_r [{shapeType=Capsule, length=2.4885, radius=8.068, shapeName=dyn_pants_02_l_capsule, center=<5.2434 0 0>, rotation=<rot 90 0 0>}]
}
constraints [
 {bone1Name=dyn_pants_01_r, bone2Name=calf_r, swing1LimitDegrees=12, swing2LimitDegrees=12, twistLimitDegrees=2}
 {bone1Name=dyn_pants_02_r, bone2Name=calf_r, swing1LimitDegrees=12, swing2LimitDegrees=12, twistLimitDegrees=2}
]
