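# Physics asset for the hair and face-accessory part of the YogaPatio medium male player cosmetic, defining bodies for dynamic hair bones.
# Nine identical small capsules cover the hair chunks, anchored to one large head capsule.
# The left and right A chunks form two-link chains with tight swing limits, while B, C and back chunks attach directly with looser limits.
# The right A root locks twist, unlike its left counterpart.
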PhysicsAsset /BRCosmetics/Characters/Player/Male/Medium/Bodies/M_MED_YogaPatio/Meshes/Parts/M_MED_YogaPatioHair_FaceAcc_Physics
bodies {
 dyn_ss_hair_chunk_A_01_l [{shapeType=Capsule, length=4, radius=3, shapeName=dyn_hair_chunk_up_B_02_l_capsule, center=<1.4559 0 0>, rotation=<rot 90 0 0>}]
 dyn_ss_hair_chunk_A_01_r [{shapeType=Capsule, length=4, radius=3, shapeName=dyn_hair_chunk_up_B_02_l_capsule, center=<1.4559 0 0>, rotation=<rot 90 0 0>}]
 dyn_ss_hair_chunk_A_02_l [{shapeType=Capsule, length=4, radius=3, shapeName=dyn_hair_chunk_up_B_02_l_capsule, center=<1.4559 0 0>, rotation=<rot 90 0 0>}]
 dyn_ss_hair_chunk_A_02_r [{shapeType=Capsule, length=4, radius=3, shapeName=dyn_hair_chunk_up_B_02_l_capsule, center=<1.4559 0 0>, rotation=<rot 90 0 0>}]
 dyn_ss_hair_chunk_B_l [{shapeType=Capsule, length=4, radius=3, shapeName=dyn_hair_chunk_up_B_02_l_capsule, center=<1.4559 0 0>, rotation=<rot 90 0 0>}]
 dyn_ss_hair_chunk_B_r [{shapeType=Capsule, length=4, radius=3, shapeName=dyn_hair_chunk_up_B_02_l_capsule, center=<1.4559 0 0>, rotation=<rot 90 0 0>}]
 dyn_ss_hair_chunk_C_l [{shapeType=Capsule, length=4, radius=3, shapeName=dyn_hair_chunk_up_B_02_l_capsule, center=<1.4559 0 0>, rotation=<rot 90 0 0>}]
 dyn_ss_hair_chunk_C_r [{shapeType=Capsule, length=4, radius=3, shapeName=dyn_hair_chunk_up_B_02_l_capsule, center=<1.4559 0 0>, rotation=<rot 90 0 0>}]
 dyn_ss_hair_chunk_bk [{shapeType=Capsule, length=4, radius=3, shapeName=dyn_hair_chunk_up_B_02_l_capsule, center=<1.4559 0 0>, rotation=<rot 90 0 0>}]
 head [{
   shapeType=Capsule
   length=13.8893
   radius=12.513
   shapeName=head_capsule
   center=<14.7463 -0.198 0.6851>
   rotation=<rot 1.4861 -102.0662 -90>
  }]
}
constraints [
 {bone1Name=dyn_ss_hair_chunk_A_02_l, bone2Name=dyn_ss_hair_chunk_A_01_l, swing1LimitDegrees=10, swing2LimitDegrees=10, twistLimitDegrees=30}
 {bone1Name=dyn_ss_hair_chunk_A_02_r, bone2Name=dyn_ss_hair_chunk_A_01_r, swing1LimitDegrees=10, swing2LimitDegrees=10, twistLimitDegrees=30}
 {bone1Name=dyn_ss_hair_chunk_A_01_l, bone2Name=head, swing1LimitDegrees=2, swing2LimitDegrees=6, twistLimitDegrees=30}
 {bone1Name=dyn_ss_hair_chunk_B_l, bone2Name=head, swing1LimitDegrees=30, swing2LimitDegrees=30, twistLimitDegrees=30}
 {bone1Name=dyn_ss_hair_chunk_A_01_r, bone2Name=head, swing1LimitDegrees=2, swing2LimitDegrees=6, twistLimitDegrees=2, twistMotion=Locked}
 {bone1Name=dyn_ss_hair_chunk_bk, bone2Name=head, swing1LimitDegrees=30, swing2LimitDegrees=30, twistLimitDegrees=30}
 {bone1Name=dyn_ss_hair_chunk_B_r, bone2Name=head, swing1LimitDegrees=30, swing2LimitDegrees=30, twistLimitDegrees=30}
 {bone1Name=dyn_ss_hair_chunk_C_l, bone2Name=head, swing1LimitDegrees=30, swing2LimitDegrees=30, twistLimitDegrees=30}
 {bone1Name=dyn_ss_hair_chunk_C_r, bone2Name=head, swing1LimitDegrees=30, swing2LimitDegrees=30, twistLimitDegrees=30}
]
