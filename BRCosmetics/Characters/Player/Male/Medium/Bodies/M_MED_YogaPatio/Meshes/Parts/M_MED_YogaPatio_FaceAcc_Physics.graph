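# Physics asset for the YogaPatio male character's face accessory mesh, giving its hair bones constrained physical motion around the head.
# A head capsule anchors short two-bone capsule chains for upper hair chunks and lower hair, in left, right and back sets.
# Each chain's root is constrained to the head and its tip to the root, all with tight swing and twist limits that keep movement subtle.
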PhysicsAsset /BRCosmetics/Characters/Player/Male/Medium/Bodies/M_MED_YogaPatio/Meshes/Parts/M_MED_YogaPatio_FaceAcc_Physics
bodies {
 dyn_hair_chunk_up_A_02_l [{
   shapeType=Capsule
   length=4
   radius=3
   shapeName=dyn_hair_chunk_up_A_02_l_capsule
   center=<0.8322 -2.3764 0>
   rotation=<rot 90 -176.4237 -161.7178>
  }]
 dyn_hair_chunk_up_A_02_r [{shapeType=Capsule, length=4, radius=3, shapeName=dyn_hair_chunk_up_A_02_r_capsule, center=<1.4559 -2.4569 0>, rotation=<rot 90 0 0>}]
 dyn_hair_chunk_up_A_03_l [{shapeType=Capsule, length=4, radius=3, shapeName=dyn_hair_chunk_up_A_03_l_capsule, center=<0.8322 -2.3764 0>, rotation=<rot 90 0 14.7058>}]
 dyn_hair_chunk_up_A_03_r [{shapeType=Capsule, length=4, radius=3, shapeName=dyn_hair_chunk_up_A_03_r_capsule, center=<1.4559 -2.4569 0>, rotation=<rot 90 0 0>}]
 dyn_hair_chunk_up_B_02_l [{shapeType=Capsule, length=4, radius=3, shapeName=dyn_hair_chunk_up_B_02_l_capsule, center=<1.4559 -2.4569 0>, rotation=<rot 90 0 0>}]
 dyn_hair_chunk_up_B_02_r [{shapeType=Capsule, length=4, radius=3, shapeName=dyn_hair_chunk_up_B_02_r_capsule, center=<1.4559 -2.4569 0>, rotation=<rot 90 0 0>}]
 dyn_hair_chunk_up_B_03_l [{shapeType=Capsule, length=4, radius=3, shapeName=dyn_hair_chunk_up_B_03_l_capsule, center=<1.4559 -2.4569 0>, rotation=<rot 90 0 0>}]
 dyn_hair_chunk_up_B_03_r [{shapeType=Capsule, length=4, radius=3, shapeName=dyn_hair_chunk_up_B_03_r_capsule, center=<1.4559 -2.4569 0>, rotation=<rot 90 0 0>}]
 dyn_hair_chunk_up_C_02_l [{shapeType=Capsule, length=4, radius=3, shapeName=dyn_hair_chunk_up_C_02_l_capsule, center=<1.4559 -2.4569 0>, rotation=<rot 90 0 0>}]
 dyn_hair_chunk_up_C_02_r [{shapeType=Capsule, length=4, radius=3, shapeName=dyn_hair_chunk_up_C_02_r_capsule, center=<1.4559 -2.4569 0>, rotation=<rot 90 0 0>}]
 dyn_hair_chunk_up_C_03_l [{shapeType=Capsule, length=4, radius=3, shapeName=dyn_hair_chunk_up_C_03_l_capsule, center=<1.4559 -2.4569 0>, rotation=<rot 90 0 0>}]
 dyn_hair_chunk_up_C_03_r [{shapeType=Capsule, length=4, radius=3, shapeName=dyn_hair_chunk_up_C_03_r_capsule, center=<1.4559 -2.4569 0>, rotation=<rot 90 0 0>}]
 dyn_hair_chunk_up_bk_01 [{shapeType=Capsule, length=4, radius=3, shapeName=dyn_hair_chunk_up_bk_01_capsule, center=<1.4559 -2.4569 0>, rotation=<rot 90 0 0>}]
 dyn_hair_chunk_up_bk_02 [{shapeType=Capsule, length=4, radius=3, shapeName=dyn_hair_chunk_up_bk_02_capsule, center=<1.4559 -2.4569 0>, rotation=<rot 90 0 0>}]
 dyn_hair_low_A_01_l [{shapeType=Capsule, length=4, radius=3, shapeName=dyn_hair_low_A_01_l_capsule, center=<1.4559 0 0>, rotation=<rot 90 0 0>}]
 dyn_hair_low_A_01_r [{shapeType=Capsule, length=4, radius=3, shapeName=dyn_hair_low_A_01_r_capsule, center=<1.4559 0 0>, rotation=<rot 90 0 0>}]
 dyn_hair_low_A_02_l [{shapeType=Capsule, length=4, radius=3, shapeName=dyn_hair_low_A_02_l_capsule, center=<1.4559 0 0>, rotation=<rot 90 0 0>}]
 dyn_hair_low_A_02_r [{shapeType=Capsule, length=4, radius=3, shapeName=dyn_hair_low_A_02_r_capsule, center=<1.4559 0 0>, rotation=<rot 90 0 0>}]
 dyn_hair_low_B_01_l [{shapeType=Capsule, length=4, radius=3, shapeName=dyn_hair_low_B_01_l_capsule, center=<1.4559 0 0>, rotation=<rot 90 0 0>}]
 dyn_hair_low_B_01_r [{shapeType=Capsule, length=4, radius=3, shapeName=dyn_hair_low_B_01_r_capsule, center=<1.4559 0 0>, rotation=<rot 90 0 0>}]
 dyn_hair_low_B_02_l [{shapeType=Capsule, length=4, radius=3, shapeName=dyn_hair_low_B_02_l_capsule, center=<1.4559 0 0>, rotation=<rot 90 0 0>}]
 dyn_hair_low_B_02_r [{shapeType=Capsule, length=4, radius=3, shapeName=dyn_hair_low_B_02_r_capsule, center=<1.4559 0 0>, rotation=<rot 90 0 0>}]
 dyn_hair_low_bk_01 [{shapeType=Capsule, length=4, radius=3, shapeName=dyn_hair_low_bk_01_capsule, center=<1.4559 0 0>, rotation=<rot 90 0 0>}]
 dyn_hair_low_bk_02 [{shapeType=Capsule, length=4, radius=3, shapeName=dyn_hair_low_bk_02_capsule, center=<1.4559 0 0>, rotation=<rot 90 0 0>}]
 head [{shapeType=Capsule, length=14.8083, radius=8.6584, shapeName=head_capsule, center=<5.3355 0 0>, rotation=<rot 90 0 0>}]
}
constraints [
 {bone1Name=dyn_hair_chunk_up_A_03_l, bone2Name=dyn_hair_chunk_up_A_02_l, swing1LimitDegrees=8, swing2LimitDegrees=8, twistLimitDegrees=2}
 {bone1Name=dyn_hair_chunk_up_A_03_r, bone2Name=dyn_hair_chunk_up_A_02_r, swing1LimitDegrees=8, swing2LimitDegrees=8, twistLimitDegrees=2}
 {bone1Name=dyn_hair_chunk_up_B_03_l, bone2Name=dyn_hair_chunk_up_B_02_l, swing1LimitDegrees=8, swing2LimitDegrees=8, twistLimitDegrees=2}
 {bone1Name=dyn_hair_chunk_up_B_03_r, bone2Name=dyn_hair_chunk_up_B_02_r, swing1LimitDegrees=8, swing2LimitDegrees=8, twistLimitDegrees=2}
 {bone1Name=dyn_hair_chunk_up_C_03_l, bone2Name=dyn_hair_chunk_up_C_02_l, swing1LimitDegrees=8, swing2LimitDegrees=8, twistLimitDegrees=2}
 {bone1Name=dyn_hair_chunk_up_C_03_r, bone2Name=dyn_hair_chunk_up_C_02_r, swing1LimitDegrees=8, swing2LimitDegrees=8, twistLimitDegrees=2}
 {bone1Name=dyn_hair_chunk_up_bk_02, bone2Name=dyn_hair_chunk_up_bk_01, swing1LimitDegrees=8, swing2LimitDegrees=8, twistLimitDegrees=2}
 {bone1Name=dyn_hair_low_A_02_l, bone2Name=dyn_hair_low_A_01_l, swing1LimitDegrees=6, swing2LimitDegrees=6, twistLimitDegrees=2}
 {bone1Name=dyn_hair_low_A_02_r, bone2Name=dyn_hair_low_A_01_r, swing1LimitDegrees=6, swing2LimitDegrees=6, twistLimitDegrees=2}
 {bone1Name=dyn_hair_low_B_02_l, bone2Name=dyn_hair_low_B_01_l, swing1LimitDegrees=6, swing2LimitDegrees=6, twistLimitDegrees=2}
 {bone1Name=dyn_hair_low_B_02_r, bone2Name=dyn_hair_low_B_01_r, swing1LimitDegrees=6, swing2LimitDegrees=6, twistLimitDegrees=2}
 {bone1Name=dyn_hair_low_bk_02, bone2Name=dyn_hair_low_bk_01, swing1LimitDegrees=6, swing2LimitDegrees=6, twistLimitDegrees=2}
 {bone1Name=dyn_hair_chunk_up_A_02_l, bone2Name=head, swing1LimitDegrees=6, swing2LimitDegrees=4, twistLimitDegrees=2}
 {bone1Name=dyn_hair_chunk_up_B_02_l, bone2Name=head, swing1LimitDegrees=6, swing2LimitDegrees=4, twistLimitDegrees=2}
 {bone1Name=dyn_hair_chunk_up_C_02_l, bone2Name=head, swing1LimitDegrees=6, swing2LimitDegrees=4, twistLimitDegrees=2}
 {bone1Name=dyn_hair_chunk_up_A_02_r, bone2Name=head, swing1LimitDegrees=6, swing2LimitDegrees=4, twistLimitDegrees=2}
 {bone1Name=dyn_hair_chunk_up_B_02_r, bone2Name=head, swing1LimitDegrees=6, swing2LimitDegrees=4, twistLimitDegrees=2}
 {bone1Name=dyn_hair_chunk_up_C_02_r, bone2Name=head, swing1LimitDegrees=6, swing2LimitDegrees=4, twistLimitDegrees=2}
 {bone1Name=dyn_hair_chunk_up_bk_01, bone2Name=head, swing1LimitDegrees=6, swing2LimitDegrees=4, twistLimitDegrees=2}
 {bone1Name=dyn_hair_low_A_01_r, bone2Name=head, swing1LimitDegrees=6, swing2LimitDegrees=6, twistLimitDegrees=2}
 {bone1Name=dyn_hair_low_B_01_r, bone2Name=head, swing1LimitDegrees=6, swing2LimitDegrees=6, twistLimitDegrees=2}
 {bone1Name=dyn_hair_low_B_01_l, bone2Name=head, swing1LimitDegrees=6, swing2LimitDegrees=6, twistLimitDegrees=2}
 {bone1Name=dyn_hair_low_A_01_l, bone2Name=head, swing1LimitDegrees=6, swing2LimitDegrees=6, twistLimitDegrees=2}
 {bone1Name=dyn_hair_low_bk_01, bone2Name=head, swing1LimitDegrees=6, swing2LimitDegrees=6, twistLimitDegrees=2}
]
